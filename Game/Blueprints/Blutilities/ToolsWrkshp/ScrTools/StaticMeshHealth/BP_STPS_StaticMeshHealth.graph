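# Editor utility blueprint that serves as a front end for a static mesh health tool, relaying test and navigation requests.
# Run Tests and Next Static Mesh only broadcast their event dispatchers, leaving the actual logic to external listeners.
# An enum variable selects which health collection to cycle through, defaulting to Severe.
# A dispatcher for clearing the health collections is declared, but nothing here calls it.
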EditorUtilityBlueprint /Game/Blueprints/Blutilities/ToolsWrkshp/ScrTools/StaticMeshHealth/BP_STPS_StaticMeshHealth
graphs [
 EventGraph {links [], nodes []}
 Next Static Mesh {nodes [
   Next Static Mesh() {is=title K2Node_FunctionEntry}
   > Call On Step to Next Problem Mesh Clicked!(execute:Exec, self:BP STPS Static Mesh Health Object Reference) -> (then:Exec) {is=title K2Node_CallDelegate}
  ]}
 OnClearStaticMeshHealthCollectionsClicked {nodes [OnClearStaticMeshHealthCollectionsClicked() {is=title K2Node_FunctionEntry}]}
 OnRunTestsClicked {nodes [OnRunTestsClicked() {is=title K2Node_FunctionEntry}]}
 OnStepToNextProblemMeshClicked {nodes [OnStepToNextProblemMeshClicked() {is=title K2Node_FunctionEntry}]}
 Run Tests {nodes [
   Run Tests() {is=title K2Node_FunctionEntry}
   > Call On Run Tests Clicked!(execute:Exec, self:BP STPS Static Mesh Health Object Reference) -> (then:Exec) {is=title K2Node_CallDelegate}
  ]}
]
variables [
 Assets to Cycle Through {type=E_CollectionChoices_StaticMeshHealth, default=Severe}
 OnClearStaticMeshHealthCollectionsClicked {default=()}
 OnRunTestsClicked {default=()}
 OnStepToNextProblemMeshClicked {default=()}
]
