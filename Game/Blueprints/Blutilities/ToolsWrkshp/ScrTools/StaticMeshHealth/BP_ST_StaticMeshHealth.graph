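# Editor scriptable tool that audits the health of static meshes in the main Valkyrie project.
# Run Tests applies Nanite, vertex, LOD, collision and material-slot checks, merging results into one per-mesh error map.
# Error counts sort meshes into Severe, Moderate and Minor Content Browser collections under a shared parent.
# Next Static Mesh cycles the chosen collection, syncs the browser and rebuilds the header with bound fix buttons.
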EditorUtilityBlueprint /Game/Blueprints/Blutilities/ToolsWrkshp/ScrTools/StaticMeshHealth/BP_ST_StaticMeshHealth
graphs [
 AddActionsBox {nodes [
   AddActionsBox() {is=title K2Node_FunctionEntry}
   > n1 = AddExpandableBox(self=Get Header() -> (Header:BP UW Shared Header Widget Object Reference))
   > n2 = Set Actions Box(Actions Box=@n1)
   > SetUpBox(self=@n2, Name=Get Actions Box Label(self=Get Header()))
  ]}
 AddNextStaticMeshButton {nodes [
   AddNextStaticMeshButton() {is=title K2Node_FunctionEntry}
   > AddChildToVerticalBox(
    self=Get DynamicEntries(self=Get Actions Box())
    Content=Get NextStaticMeshButton() -> (NextStaticMeshButton:BP UW Next Static Mesh Button Object Reference)
   )
   > Bind Event to Next Static Mesh Clicked(self:BP UW Next Static Mesh Button Object Reference=Get NextStaticMeshButton(), Delegate=Create Event()) {is=title K2Node_AddDelegate}
  ]}
 AddRunTestsButton {nodes [
   AddRunTestsButton() {is=title K2Node_FunctionEntry}
   > AddChildToVerticalBox(self=Get DynamicEntries(self=Get Actions Box()), Content=Get RunTestsButton() -> (RunTestsButton:BP UW Run Tests Button Object Reference))
   > Bind Event to Run Tests Clicked(self:BP UW Run Tests Button Object Reference=Get RunTestsButton(), Delegate=Create Event()) {is=title K2Node_AddDelegate}
  ]}
 Destroy Collection {nodes [
   n1 = Destroy Collection!() -> (then, Collection Name:String, ShareType:ECollectionScriptingShareType Enum=Local) {is=title K2Node_FunctionEntry}
   > n2 = GetMainProject!(execute) -> (then, ReturnValue:Valkyrie Project Handle Structure)
   n3 = GetProjectCollectionContainer!(ProjectHandle:"Valkyrie Project Handle Structure (by ref)"=@n2) -> (OutProjectCollectionContainer:Collection Container Source Structure, ReturnValue:bool)
   > n4 = DestroyCollection!(
    execute
    self:Collection Manager Subsystem Object Reference=Get CollectionManagerScriptingSubsystem()
    Collection:"Collection Structure (by ref)"=Make Collection!(
     Container:Name=Break Collection Container Source!(CollectionScriptingContainerSource:"Collection Container Source Structure (by ref)"=@n3) -> (Name:Name, Title:Text) {is=title K2Node_BreakStruct}
     Name:Name=String To Name(InString=@n1)
     ShareType:ECollectionScriptingShareType Enum=@n1.ShareType
    ) -> (CollectionScriptingRef:Collection Structure) {is=title K2Node_MakeStruct}
   ) -> (then, ReturnValue:bool)
   > Return Node!(execute, Success:bool=@n4)
  ]}
 EventGraph {nodes [
   n1 = SyncBrowserToAsset!() -> (OutputDelegate, then, TheObject:Object Reference) {is=title K2Node_CustomEvent}
   > Delay(Duration=0.2)
   > SyncBrowserToObjects(AssetPaths=Make Array("[0]":String=GetSoftObjectPath!(Object=@n1.TheObject) -> (ReturnValue_PathString:String)) -> (Array:Array of Strings))
   Event OnScriptSetup()
   > n2 = Sequence!(execute) -> (then_0, then_1, then_2:Exec, then_3:Exec, then_4:Exec, then_5:Exec, then_6:Exec)
   > K2Node_Composite_0 = Init Property Sets() {is=title K2Node_Composite}
   K2Node_Composite_1 = Init User Widgets(execute=@n2.then_1) {is=title K2Node_Composite}
   K2Node_Composite_2 = Bind Button Events!(execute=@n2.then_2) -> (then:Exec) {is=title K2Node_Composite}
   K2Node_Composite_3 = Watch Properties(execute=@n2.then_3) {is=title K2Node_Composite}
   K2Node_Composite_4 = Populate Project Vars(execute=@n2.then_4) {is=title K2Node_Composite}
   Populate Collection Vars(execute=@n2.then_5)
   SetFocusInViewport(execute=@n2.then_6, self:Scriptable Interactive Tool Object Reference)
   Event OnScriptShutdown()
   > SavePropertySetSettings(
    PropertySet=Get Settings() -> (Settings:BP STPS Static Mesh Health Object Reference)
    SaveKey=Get Prop Set Save Key() -> (Prop Set Save Key:String)
   )
   n3 = Event OnScriptTick()
   > Set CurTime(CurTime=float + float(A=@n3.DeltaTime, B=Get CurTime()))
   > n4 = Branch(Condition=Get FirstInitRun())
   Set FirstInitRun(execute=@n4.else)
   > AddActionsBox() {is=title K2Node_CallFunction}
   > n5 = Sequence!(execute) -> (then_0, then_1)
   > n6 = Construct Object from Class(Class=/Game/Blueprints/Blutilities/ToolsWrkshp/ScrTools/StaticMeshHealth/SupportFiles/BP_UW_RunTestsButton.BP_UW_RunTestsButton_C) -> (ReturnValue:BP UW Run Tests Button Object Reference)
   > Set RunTestsButton(RunTestsButton:BP UW Run Tests Button Object Reference=@n6) -> (Output_Get:BP UW Run Tests Button Object Reference)
   > AddRunTestsButton() {is=title K2Node_CallFunction}
   n7 = Construct Object from Class(
    execute=@n5.then_1
    Class=/Game/Blueprints/Blutilities/ToolsWrkshp/ScrTools/StaticMeshHealth/SupportFiles/BP_UW_NextStaticMeshButton.BP_UW_NextStaticMeshButton_C
   ) -> (ReturnValue:BP UW Next Static Mesh Button Object Reference)
   > Set NextStaticMeshButton(NextStaticMeshButton:BP UW Next Static Mesh Button Object Reference=@n7) -> (Output_Get:BP UW Next Static Mesh Button Object Reference)
   > AddNextStaticMeshButton() {is=title K2Node_CallFunction}
   > NextButtonEnable!(execute, self, Enable:bool=false) -> (then)
  ]}
 Init Property Sets {
  path=EventGraph.K2Node_Composite_0.Init Property Sets
  nodes [
   Inputs()
   > n1 = AddPropertySetOfType(
    PropertySetType=&/BP_STPS_StaticMeshHealth.BP_STPS_StaticMeshHealth_C
    Identifier=Settings
   )
   > n2 = Cast To BP_STPS_StaticMeshHealth!(execute, Object=@n1) -> (then, CastFailed, AsBP STPS Static Mesh Health:BP STPS Static Mesh Health Object Reference) {is=title K2Node_DynamicCast}
   > n3 = Set Settings(Settings:BP STPS Static Mesh Health Object Reference=@n2) -> (Output_Get:BP STPS Static Mesh Health Object Reference)
   > RestorePropertySetSettings(PropertySet=@n3, SaveKey=Get Prop Set Save Key())
   > DisplayUserWarningMessage()
   Outputs()
  ]
 }
 Init User Widgets {
  path=EventGraph.K2Node_Composite_1.Init User Widgets
  nodes [
   Inputs()
   > n1 = Construct Object from Class(Class=&BP_UW_SharedHeaderWidget_C) -> (ReturnValue:BP UW Shared Header Widget Object Reference)
   > n2 = Set Header(Header:BP UW Shared Header Widget Object Reference=@n1) -> (Output_Get:BP UW Shared Header Widget Object Reference)
   > SetToolPanelToolHeaderWidget(Widget=@n2)
   Outputs()
  ]
 }
 Bind Button Events {
  path=EventGraph.K2Node_Composite_2.Bind Button Events
  nodes [
   Inputs()
   > Sequence()
   > Bind Event to On Run Tests Clicked(self:BP STPS Static Mesh Health Object Reference=Get Settings(), Delegate=Create Event()) {is=title K2Node_AddDelegate}
   > Bind Event to On Step to Next Problem Mesh Clicked(self:BP STPS Static Mesh Health Object Reference=Get Settings(), Delegate=Create Event()) {is=title K2Node_AddDelegate}
   > Outputs!(then:Exec)
  ]
 }
 Watch Properties {
  path=EventGraph.K2Node_Composite_3.Watch Properties
  nodes [
   Inputs()
   > WatchProperty(PropertySet=Get Settings(), PropertyName=Assets to Cycle Through, OnModified=Create Event())
   Outputs!()
  ]
 }
 Populate Project Vars {
  path=EventGraph.K2Node_Composite_4.Populate Project Vars
  nodes [
   Inputs()
   > n1 = GetMainProject()
   > n2 = Set Main Project Handle(Main Project Handle:Valkyrie Project Handle Structure=@n1) -> (Output_Get:Valkyrie Project Handle Structure)
   > Set Project Handle Valid(Project Handle Valid:bool=IsProjectHandleValid!(ProjectHandle:"Valkyrie Project Handle Structure (by ref)"=@n2) -> (ReturnValue:bool)) -> (Output_Get:bool)
   > Set Project Asset Path(Project Asset Path:String=GetProjectMountedAssetPath!(ProjectHandle:"Valkyrie Project Handle Structure (by ref)"=@n2) -> (ReturnValue:String)) -> (Output_Get:String)
   > Set Valkyrie Main Proj Collection Container(Valkyrie Main Proj Collection Container:Collection Container Source Structure=GetProjectCollectionContainer(ProjectHandle=Get Main Project Handle() -> (Main Project Handle:Valkyrie Project Handle Structure))) -> (Output_Get:Collection Container Source Structure)
   Outputs()
  ]
 }
 Make Collection {nodes [
   n1 = Make Collection!() -> (then, Collection Name:String, Parent Collection Name:String, ShareType:ECollectionScriptingShareType Enum=Local) {is=title K2Node_FunctionEntry}
   > n2 = Sequence()
   > n3 = CollectionExists!(
    execute
    self:Collection Manager Subsystem Object Reference=Get CollectionManagerScriptingSubsystem()
    Container:Collection Container Source Structure=Get Valkyrie Main Proj Collection Container() -> (Valkyrie Main Proj Collection Container:Collection Container Source Structure)
    Collection:Name=String To Name(InString=@n1.Parent Collection Name)
    ShareType:ECollectionScriptingShareType Enum=Get ShareType() -> (ShareType:ECollectionScriptingShareType Enum=Local)
   ) -> (then, ReturnValue:bool)
   > n4 = Branch(Condition=@n3)
   n5 = CreateCollection!(
    execute=@n4.else
    self:Collection Manager Subsystem Object Reference=Get CollectionManagerScriptingSubsystem()
    Container:Collection Container Source Structure=Get Valkyrie Main Proj Collection Container()
    Collection:Name=String To Name(InString=Get Parent Collection Name() -> (Parent Collection Name:String))
    ShareType:ECollectionScriptingShareType Enum=Get ShareType() -> (ShareType=Local)
   ) -> (then, OutNewCollection:Collection Structure, ReturnValue:bool)
   > Set Parent Collection(Parent Collection:Collection Structure=@n5) -> (Output_Get:Collection Structure)
   n6 = CreateOrEmptyCollection!(
    execute=@n2.then_1
    self:Collection Manager Subsystem Object Reference=Get CollectionManagerScriptingSubsystem()
    Container:Collection Container Source Structure=Get Valkyrie Main Proj Collection Container()
    Collection:Name=String To Name(InString=@n1)
    ShareType:ECollectionScriptingShareType Enum=Get ShareType() -> (ShareType=Local)
   ) -> (then, OutNewOrEmptyCollection:Collection Structure, ReturnValue:bool)
   > ReparentCollection!(
    execute
    self:Collection Manager Subsystem Object Reference=Get CollectionManagerScriptingSubsystem()
    Collection:"Collection Structure (by ref)"=@n6
    NewParentCollection:Collection Structure=Get Parent Collection() -> (Parent Collection:Collection Structure)
   ) -> (then, ReturnValue:bool)
   > Return Node!(execute, OutNewOrEmptyCollection:Collection Structure=@n6, Success=@n6.ReturnValue)
  ]}
 Mesh Map Merge {nodes [
   n1 = Mesh Map Merge!() -> (
    then
    Array 1:Map of Static Mesh Object References to S Asset Errors Asset Health Structures
    Merge Dest Array:Map of Static Mesh Object References to S Asset Errors Asset Health Structures
   ) {is=title K2Node_FunctionEntry}
   > Set Temp Merged(Temp Merged:Map of Static Mesh Object References to S Asset Errors Asset Health Structures=@n1.Merge Dest Array) -> (Output_Get:Map of Static Mesh Object References to S Asset Errors Asset Health Structures)
   > n2 = "For Each Loop (Map)"!(
    execute:Exec
    MapPin:Map of Static Mesh Object References to S Asset Errors Asset Health Structures=Get Array 1() -> (Array 1:Map of Static Mesh Object References to S Asset Errors Asset Health Structures)
    BreakPin:Exec
   ) -> (then:Exec, KeyPin:Static Mesh Object Reference, ValuePin:S Asset Errors Asset Health Structure, CompletedPin:Exec)
   > Set Current StaticMesh(Current StaticMesh:Static Mesh Object Reference=@n2) -> (Output_Get:Static Mesh Object Reference)
   > Set Array 1 Value(Array 1 Value:S Asset Errors Asset Health Structure=Find(
     TargetMap:Map of Static Mesh Object References to S Asset Errors Asset Health Structures=Get Array 1()
     Key:"Static Mesh Object Reference (by ref)"=Get Current StaticMesh() -> (Current StaticMesh:Static Mesh Object Reference)
    ) -> (Value:S Asset Errors Asset Health Structure)) -> (Output_Get:S Asset Errors Asset Health Structure)
   n3 = Find(
    TargetMap=Get Temp Merged() -> (Temp Merged:Map of Static Mesh Object References to S Asset Errors Asset Health Structures)
    Key=Get Current StaticMesh()
   )
   > Set Found In Temp Merged(Found In Temp Merged:S Asset Errors Asset Health Structure=@n3) -> (Output_Get:S Asset Errors Asset Health Structure)
   > n4 = Branch(Condition=@n3.ReturnValue)
   n5 = Break S Asset Errors Asset Health!(S_AssetErrors_AssetHealth:"S Asset Errors Asset Health Structure (by ref)"=Find(TargetMap=Get Temp Merged(), Key=Get Current StaticMesh())) -> (
    Errors_3_#1:Array of Texts
    Fixes_14_#2:Map of Strings to Strings
    ErrorTooltips_19_#3:Array of Texts
   ) {is=title K2Node_BreakStruct}
   > Set Temp Merged Messages(Temp Merged Messages:Array of Texts=@n5) -> (Output_Get:Array of Texts)
   > Set Temp Merged Fixes(Temp Merged Fixes:Map of Strings to Strings=@n5.Fixes_14_#2) -> (Output_Get:Map of Strings to Strings)
   > Set Temp Merged Tooltips(Temp Merged Tooltips:Array of Texts=@n5.ErrorTooltips_19_#3) -> (Output_Get:Array of Texts)
   n6 = Break S Asset Errors Asset Health(S_AssetErrors_AssetHealth=Get Array 1 Value() -> (Array 1 Value:S Asset Errors Asset Health Structure))
   > Set Temp Array 1 Fixes(Temp Array 1 Fixes:Map of Strings to Strings=@n6.Fixes_14_#2) -> (Output_Get:Map of Strings to Strings)
   > Set Array 1 Tooltips(Array 1 Tooltips:Array of Texts=@n6.ErrorTooltips_19_#3) -> (Output_Get:Array of Texts)
   > n7 = For Each Loop(Array:Array of Texts=@n6) -> (Array Element:Text)
   n8 = For Each Loop(Exec=@n7.Completed, Array=Get Temp Merged Messages() -> (Temp Merged Messages:Array of Texts))
   Remove!(
    execute=@n8.Completed
    TargetMap:Map of Static Mesh Object References to S Asset Errors Asset Health Structures=Get Temp Merged()
    Key:"Static Mesh Object Reference (by ref)"=Get Current StaticMesh()
   ) -> (then, ReturnValue:bool)
   > Add!(
    execute
    TargetMap:Map of Static Mesh Object References to S Asset Errors Asset Health Structures=Get Temp Merged()
    Key:"Static Mesh Object Reference (by ref)"=Get Current StaticMesh()
    Value:"S Asset Errors Asset Health Structure (by ref)"=Make S Asset Errors Asset Health!(
     Errors_3_#1:Array of Texts=Get Temp Merged Messages()
     Fixes_14_#2:Map of Strings to Strings=Get Temp New Fixes() -> (Temp New Fixes:Map of Strings to Strings)
     ErrorTooltips_19_#3:Array of Texts=Get Temp Merged Tooltips() -> (Temp Merged Tooltips:Array of Texts)
    ) -> (S_AssetErrors_AssetHealth:S Asset Errors Asset Health Structure) {is=title K2Node_MakeStruct}
   ) -> (then) {is=title K2Node_CallFunction}
   n9 = "To String (Text)"(InText=@n8)
   n10 = Find(TargetMap:Map of Strings to Strings=Get Temp Array 1 Fixes() -> (Temp Array 1 Fixes:Map of Strings to Strings), Key:"String (by ref)"=@n9) -> (Value:String)
   n11 = Branch(execute=@n8.LoopBody, Condition=@n10.ReturnValue)
   > Add(TargetMap:Map of Strings to Strings=Get Temp New Fixes(), Key:"String (by ref)"=@n9, Value:"String (by ref)"=@n10)
   n12 = Find(TargetMap=Get Temp Merged Fixes() -> (Temp Merged Fixes:Map of Strings to Strings), Key=@n9)
   Branch(execute=@n11.else, Condition=@n12.ReturnValue)
   > Add(TargetMap=Get Temp New Fixes(), Key=@n9, Value=@n12)
   Add Unique(execute=@n7.LoopBody, TargetArray:Array of Texts=Get Temp Merged Messages(), NewItem:"Text (by ref)"=@n7)
   > Add(
    TargetArray:Array of Texts=Get Temp Merged Tooltips()
    NewItem:"Text (by ref)"="Get (a ref)"(Array:Array of Texts=Get Array 1 Tooltips() -> (Array 1 Tooltips:Array of Texts), Dimension 1=@n7.Array Index) -> (Output:"Text (by ref)")
   ) {is=title K2Node_CallArrayFunction}
   Add(
    execute=@n4.else
    TargetMap:Map of Static Mesh Object References to S Asset Errors Asset Health Structures=Get Temp Merged()
    Key:"Static Mesh Object Reference (by ref)"=Get Current StaticMesh()
    Value:"S Asset Errors Asset Health Structure (by ref)"=Get Array 1 Value()
   ) {is=title K2Node_CallFunction}
   Return Node!(execute=@n2.CompletedPin, Merged:Map of Static Mesh Object References to S Asset Errors Asset Health Structures=Get Temp Merged())
  ]}
 NextButtonEnable {nodes [
   n1 = NextButtonEnable!() -> (then, Enable:bool=true) {is=title K2Node_FunctionEntry}
   > SetIsEnabled(self=Get NextStaticMeshButton(), bInIsEnabled=@n1)
   n2 = SelectText(A=Get NextButtonEnabledText() -> (NextButtonEnabledText:Text), B=Get NextButtonDisabledText() -> (NextButtonDisabledText:Text), bPickA=@n1)
   n3 = Get NextStaticMeshButton()
   > Set ToolTipText!(execute, ToolTipText:Text=@n2, self:Widget Object Reference=@n3) -> (then, Output_Get:Text)
   > SetToolTipText!(execute, self:Widget Object Reference=@n3, InToolTipText:"Text (by ref)"=@n2) -> (then)
  ]}
 On Next Prob Mesh Clicked {nodes [
   On Next Prob Mesh Clicked() {is=title K2Node_FunctionEntry}
   n1 = Select!(
    NewEnumerator0:Collection Structure=Get Severe Collection() -> (Severe Collection:Collection Structure)
    NewEnumerator1:Collection Structure=Get Moderate Collection() -> (Moderate Collection:Collection Structure)
    NewEnumerator2:Collection Structure=Get Minor Collection() -> (Minor Collection:Collection Structure)
    Index:E_CollectionChoices_StaticMeshHealth Enum=Get Assets to Cycle Through!(self:BP STPS Static Mesh Health Object Reference=Get Settings()) -> (Assets to Cycle Through:E_CollectionChoices_StaticMeshHealth Enum=NewEnumerator0)
   ) -> (ReturnValue:Collection Structure)
   > n2 = CollectionExists(
    self=Get CollectionManagerScriptingSubsystem()
    Container=Get Valkyrie Main Proj Collection Container()
    Collection=Break Collection!(CollectionScriptingRef:"Collection Structure (by ref)"=@n1) -> (Container:Name, Name:Name, ShareType:ECollectionScriptingShareType Enum) {is=title K2Node_BreakStruct}.Name
    ShareType=Get Collection Share Type() -> (Collection Share Type:ECollectionScriptingShareType Enum=Local)
   )
   > n3 = Branch(Condition=@n2)
   > n4 = GetAssetsInCollection!(
    execute
    self:Collection Manager Subsystem Object Reference=Get CollectionManagerScriptingSubsystem()
    Collection:"Collection Structure (by ref)"=@n1
   ) -> (then, OutAssets:Array of Asset Data Structures, ReturnValue:bool)
   > n5 = Set Num Assets In This Collection(Num Assets In This Collection:int=Length(TargetArray:Array of Asset Data Structures=@n4)) -> (Output_Get:int=0)
   > n6 = Branch(Condition=OR Boolean(A=NOT Boolean(A=@n4.ReturnValue), B="integer < integer"(A=@n5, B=1)))
   n7 = Get Assets to Cycle Through(self=Get Settings()) -> (Assets to Cycle Through=NewEnumerator0)
   > DisplayUserWarningMessage(Message=Format Text!(
     Format="LOCTEXT(\"No {s} assets found\")"
     s:String=Enum to String!(Enumerator:E_CollectionChoices_StaticMeshHealth Enum=@n7) -> (ReturnValue:String)
    ) -> (Result))
   DisplayUserWarningMessage(execute=@n6.else)
   > Branch(Condition=@n4.ReturnValue)
   n8 = Get Cur Mesh Asset Selected() -> (Cur Mesh Asset Selected:int=0)
   > n9 = Cast To StaticMesh(Object=GetAsset(InAssetData="Get (a ref)"(Array:Array of Asset Data Structures=@n4, Dimension 1=@n8) -> (Output:"Asset Data Structure (by ref)")))
   > n10 = Set Cur Static Mesh(Cur Static Mesh:Static Mesh Object Reference=@n9) -> (Output_Get:Static Mesh Object Reference)
   > n11 = Branch(Condition=Find(
     TargetMap:Map of Static Mesh Object References to S Asset Errors Asset Health Structures=Get All StaticMeshes() -> (All StaticMeshes:Map of Static Mesh Object References to S Asset Errors Asset Health Structures)
     Key:"Static Mesh Object Reference (by ref)"=Get Cur Static Mesh() -> (Cur Static Mesh:Static Mesh Object Reference)
    ) -> (Value:S Asset Errors Asset Health Structure).ReturnValue)
   > Set Currently Selected Asset(Currently Selected Asset:Static Mesh Object Reference=@n10) -> (Output_Get:Static Mesh Object Reference)
   > n12 = Increment Int(Value=Get Cur Mesh Asset Selected() -> (Cur Mesh Asset Selected=0))
   > n13 = Sequence()
   n14 = Get Num Assets In This Collection() -> (Num Assets In This Collection:int=0)
   > Branch(Condition="integer > integer"(A=@n12, B=int - int(A=@n14)))
   > Set Cur Mesh Asset Selected(Cur Mesh Asset Selected:int=0) -> (Output_Get:int=0)
   n15 = Get Cur Static Mesh()
   Set Cur Mesh For Button Fixes(execute=@n13.then_1, Cur Mesh For Button Fixes:Static Mesh Object Reference=@n15) -> (Output_Get:Static Mesh Object Reference)
   > RemoveFromParent(self=Get Actions Box())
   n16 = Find(TargetMap=Get All StaticMeshes(), Key=Get Cur Static Mesh())
   > n17 = Update Health Tool Header!(
    execute
    Header Widget:BP UW Shared Header Widget Object Reference=Get Header()
    Cur Asset:Object Reference=@n15
    Cur Asset Error Struct:S Asset Errors Asset Health Structure=@n16
   ) -> (then, Resulting Buttons:Array of BP UW Shared Fix Me Button Object References)
   > Set ThisAssetButtons(ThisAssetButtons:Array of BP UW Shared Fix Me Button Object References=@n17) -> (Output_Get:Array of BP UW Shared Fix Me Button Object References)
   > AddActionsBox()
   > AddRunTestsButton()
   > AddNextStaticMeshButton()
   > NextButtonEnable(Enable=true) {is=title K2Node_CallFunction}
   n18 = Break S Asset Errors Asset Health(S_AssetErrors_AssetHealth=@n16)
   > n19 = For Each Loop(Array=@n18)
   SyncBrowserToAsset!(execute=@n19.Completed, self, TheObject:Object Reference=Get Cur Static Mesh()) -> (then) {is=title K2Node_CallFunction}
   > Set Previously Selected Asset(Previously Selected Asset:Static Mesh Object Reference=Get Cur Static Mesh()) -> (Output_Get:Static Mesh Object Reference)
   n20 = Find(TargetMap:Map of Strings to Strings=@n18.Fixes_14_#2, Key:"String (by ref)"="To String (Text)"(InText=@n19)) -> (Value:String)
   Branch(execute=@n19.LoopBody, Condition=@n20.ReturnValue)
   > n21 = Switch on String!(execute:Exec, Selection:String=@n20) -> (
    Default:Exec
    MeshFix_AddLODs:Exec
    MeshFix_AddCollision:Exec
    MeshFix_TurnOnNanite:Exec
    MeshFix_SetMinLODforPlatform:Exec
    MeshFix_SetLODforCollision:Exec
    MeshFix_CollisionComplexity:Exec
   ) {pinNames [
     MeshFix_AddLODs
     MeshFix_AddCollision
     MeshFix_TurnOnNanite
     MeshFix_SetMinLODforPlatform
     MeshFix_SetLODforCollision
     MeshFix_CollisionComplexity
    ]}
   n22 = "Get (a copy)"(
    Array:Array of BP UW Shared Fix Me Button Object References=Get ThisAssetButtons() -> (ThisAssetButtons:Array of BP UW Shared Fix Me Button Object References)
    Dimension 1=@n19.Array Index
   ) -> (Output:BP UW Shared Fix Me Button Object Reference)
   Bind Event to Fix Me Button Pressed(execute=@n21.MeshFix_AddCollision, self:BP UW Shared Fix Me Button Object Reference=@n22, Delegate=Create Event()) {is=title K2Node_AddDelegate}
   Bind Event to Fix Me Button Pressed(execute=@n21.MeshFix_AddLODs, self=@n22, Delegate=Create Event())
   Bind Event to Fix Me Button Pressed(execute=@n21.MeshFix_CollisionComplexity, self=@n22, Delegate=Create Event())
   Bind Event to Fix Me Button Pressed(execute=@n21.MeshFix_SetLODforCollision, self=@n22, Delegate=Create Event())
   Bind Event to Fix Me Button Pressed(execute=@n21.MeshFix_SetMinLODforPlatform, self=@n22, Delegate=Create Event())
   Bind Event to Fix Me Button Pressed(execute=@n21.MeshFix_TurnOnNanite, self=@n22, Delegate=Create Event())
   DisplayUserWarningMessage(execute=@n11.else, Message="LOCTEXT(\"Click Run Tests first\")")
   > Return Node!(execute)
   n23 = Get Assets to Cycle Through(self=Get Settings()) -> (Assets to Cycle Through=NewEnumerator0)
   DisplayUserWarningMessage(execute=@n3.else, Message=Format Text(Format="LOCTEXT(\"No {s} assets found\")", s=Enum to String(Enumerator=@n23)))
  ]}
 On Run Tests Clicked {nodes [
   On Run Tests Clicked() {is=title K2Node_FunctionEntry}
   > n1 = Sequence!(execute) -> (then_0, then_1, then_2:Exec, then_3:Exec, then_4:Exec, then_5:Exec, then_6:Exec, then_7:Exec, then_8:Exec)
   > DisplayUserWarningMessage()
   > Set FinalRunMessage(FinalRunMessage:Text) -> (Output_Get:Text)
   K2Node_Composite_2 = Find all assets to check(execute=@n1.then_1) {is=title K2Node_Composite}
   K2Node_Composite_0 = Run Geo Tests(execute=@n1.then_2) {is=title K2Node_Composite}
   K2Node_Composite_3 = Sort into Severity!(InputPin:Exec=@n1.then_3) {is=title K2Node_Composite}
   K2Node_Composite_4 = Make Severe Collection!(InputPin:Exec=@n1.then_4) {is=title K2Node_Composite}
   K2Node_Composite_6 = Make Moderate Collection(execute=@n1.then_5) {is=title K2Node_Composite}
   K2Node_Composite_7 = Make Minor Collection(execute=@n1.then_6) {is=title K2Node_Composite}
   K2Node_Composite_9 = Set Menu To highest severity(execute=@n1.then_7) {is=title K2Node_Composite}
  ]}
 Run Geo Tests {
  path=On Run Tests Clicked.K2Node_Composite_0.Run Geo Tests
  nodes [
   n1 = Get Nanite Not Enabled() -> (Nanite Not Enabled:Map of Static Mesh Object References to S Asset Errors Asset Health Structures)
   Clear!(execute, TargetMap:Map of Static Mesh Object References to S Asset Errors Asset Health Structures=@n1) -> (then) {is=title K2Node_CallFunction}
   > n2 = MeshCheck_NaniteDisabled!(
    execute
    Static Mesh:Static Mesh Object Reference=Get Cur Project StaticMesh() -> (Cur Project StaticMesh:Static Mesh Object Reference)
    NaniteNotEnabledArray:Map of Static Mesh Object References to S Asset Errors Asset Health Structures=@n1
    Only Major:bool=true
   ) -> (then, Updated Array:Map of Static Mesh Object References to S Asset Errors Asset Health Structures)
   > n3 = Set Nanite Not Enabled(Nanite Not Enabled:Map of Static Mesh Object References to S Asset Errors Asset Health Structures=@n2) -> (Output_Get:Map of Static Mesh Object References to S Asset Errors Asset Health Structures)
   > n4 = Mesh Map Merge!(
    execute
    self
    Array 1:Map of Static Mesh Object References to S Asset Errors Asset Health Structures=@n3
    Merge Dest Array:Map of Static Mesh Object References to S Asset Errors Asset Health Structures=Get All StaticMeshes()
   ) -> (then, Merged:Map of Static Mesh Object References to S Asset Errors Asset Health Structures) {is=title K2Node_CallFunction}
   > Set All StaticMeshes(All StaticMeshes:Map of Static Mesh Object References to S Asset Errors Asset Health Structures=@n4) -> (Output_Get:Map of Static Mesh Object References to S Asset Errors Asset Health Structures)
   Inputs()
   > n5 = For Each Loop(Array:Array of Static Mesh Object References=Get Project Static Mesh() -> (Project Static Mesh:Array of Static Mesh Object References)) -> (Array Element:Static Mesh Object Reference)
   > Set Cur Project StaticMesh(Cur Project StaticMesh:Static Mesh Object Reference=@n5) -> (Output_Get:Static Mesh Object Reference)
   > n6 = Sequence()
   n7 = Get LOD0 Too Many Verts() -> (LOD0 Too Many Verts:Map of Static Mesh Object References to S Asset Errors Asset Health Structures)
   > Clear(TargetMap=@n7)
   > n8 = MeshCheck_TooManyVerts!(
    execute
    Static Mesh:Static Mesh Object Reference=Get Cur Project StaticMesh()
    LOD:int=0
    TooManyVertsArray:Map of Static Mesh Object References to S Asset Errors Asset Health Structures=@n7
   ) -> (then, Updated Array:Map of Static Mesh Object References to S Asset Errors Asset Health Structures)
   > n9 = Set LOD0 Too Many Verts(LOD0 Too Many Verts:Map of Static Mesh Object References to S Asset Errors Asset Health Structures=@n8) -> (Output_Get:Map of Static Mesh Object References to S Asset Errors Asset Health Structures)
   > n10 = Mesh Map Merge(Array 1=@n9, Merge Dest Array=Get All StaticMeshes())
   > Set All StaticMeshes(All StaticMeshes=@n10)
   n11 = Get ComplexAsSimple() -> (ComplexAsSimple:Map of Static Mesh Object References to S Asset Errors Asset Health Structures)
   Clear(execute=@n6.then_1, TargetMap=@n11)
   > n12 = MeshCheck_ComplexAsSimple!(
    execute
    Static Mesh:Static Mesh Object Reference=Get Cur Project StaticMesh()
    ComplexCollisionArray:Map of Static Mesh Object References to S Asset Errors Asset Health Structures=@n11
    Only Major:bool=true
   ) -> (then, Updated Array:Map of Static Mesh Object References to S Asset Errors Asset Health Structures)
   > n13 = Set ComplexAsSimple(ComplexAsSimple:Map of Static Mesh Object References to S Asset Errors Asset Health Structures=@n12) -> (Output_Get:Map of Static Mesh Object References to S Asset Errors Asset Health Structures)
   > n14 = Mesh Map Merge(Array 1=@n13, Merge Dest Array=Get All StaticMeshes())
   > Set All StaticMeshes(All StaticMeshes=@n14)
   n15 = Get Num Mat Slots Array() -> (Num Mat Slots Array:Map of Static Mesh Object References to S Asset Errors Asset Health Structures)
   Clear(execute=@n6.then_2, TargetMap=@n15)
   > n16 = MeshCheck_NumMatSlots!(
    execute
    Static Mesh:Static Mesh Object Reference=Get Cur Project StaticMesh()
    Acceptable Num Mat Slots:int=Get Acceptable Num Mat Slots() -> (Acceptable Num Mat Slots:int=0)
    LOD To Check:int=0
    NumMatSlotsArray:Map of Static Mesh Object References to S Asset Errors Asset Health Structures=@n15
    Only Major:bool=true
   ) -> (then, Updated Array:Map of Static Mesh Object References to S Asset Errors Asset Health Structures)
   > n17 = Set Num Mat Slots Array(Num Mat Slots Array:Map of Static Mesh Object References to S Asset Errors Asset Health Structures=@n16) -> (Output_Get:Map of Static Mesh Object References to S Asset Errors Asset Health Structures)
   > n18 = Mesh Map Merge(Array 1=@n17, Merge Dest Array=Get All StaticMeshes())
   > Set All StaticMeshes(All StaticMeshes=@n18)
   n19 = Get Has LODS Array() -> (Has LODS Array:Map of Static Mesh Object References to S Asset Errors Asset Health Structures)
   Clear(execute=@n6.then_3, TargetMap=@n19)
   > n20 = MeshCheck_HasLODs!(
    execute
    Static Mesh:Static Mesh Object Reference=Get Cur Project StaticMesh()
    MinLODCount:int=Get Acceptable Min LOD Count() -> (Acceptable Min LOD Count:int=0)
    HasLODS_Array:Map of Static Mesh Object References to S Asset Errors Asset Health Structures=@n19
    Only Major:bool=true
   ) -> (then, Updated Array:Map of Static Mesh Object References to S Asset Errors Asset Health Structures)
   > n21 = Set Has LODS Array(Has LODS Array:Map of Static Mesh Object References to S Asset Errors Asset Health Structures=@n20) -> (Output_Get:Map of Static Mesh Object References to S Asset Errors Asset Health Structures)
   > n22 = Mesh Map Merge(Array 1=@n21, Merge Dest Array=Get All StaticMeshes())
   > Set All StaticMeshes(All StaticMeshes=@n22)
   n23 = Get Min LODfor Low Array() -> (Min LODfor Low Array:Map of Static Mesh Object References to S Asset Errors Asset Health Structures)
   Clear(execute=@n6.then_4, TargetMap=@n23)
   > n24 = MeshCheck_MinLODforLow!(
    execute
    Static Mesh:Static Mesh Object Reference=Get Cur Project StaticMesh()
    MinLOD_ForLow:int=Get Min LOD for Low() -> (Min LOD for Low:int=0)
    MinLOD_ForMedium:int=Get Min LOD for Medium() -> (Min LOD for Medium:int=0)
    MinLODforLowArray:Map of Static Mesh Object References to S Asset Errors Asset Health Structures=@n23
    Only Major:bool=true
   ) -> (then, Updated Array:Map of Static Mesh Object References to S Asset Errors Asset Health Structures)
   > n25 = Set Min LODfor Low Array(Min LODfor Low Array:Map of Static Mesh Object References to S Asset Errors Asset Health Structures=@n24) -> (Output_Get:Map of Static Mesh Object References to S Asset Errors Asset Health Structures)
   > n26 = Mesh Map Merge(Array 1=@n25, Merge Dest Array=Get All StaticMeshes())
   > Set All StaticMeshes(All StaticMeshes=@n26)
   n27 = Get Similar LODS Array() -> (Similar LODS Array:Map of Static Mesh Object References to S Asset Errors Asset Health Structures)
   Clear(execute=@n6.then_5, TargetMap=@n27)
   > n28 = MeshCheck_SimilarLODs!(
    execute
    Static Mesh:Static Mesh Object Reference=Get Cur Project StaticMesh()
    AcceptablePercent:int=Get Acceptable Percent Diff For LODs() -> (Acceptable Percent Diff For LODs:int=0)
    SimilarLODS_Array:Map of Static Mesh Object References to S Asset Errors Asset Health Structures=@n27
    Only Major:bool=true
   ) -> (then, Updated Array:Map of Static Mesh Object References to S Asset Errors Asset Health Structures)
   > n29 = Set Similar LODS Array(Similar LODS Array:Map of Static Mesh Object References to S Asset Errors Asset Health Structures=@n28) -> (Output_Get:Map of Static Mesh Object References to S Asset Errors Asset Health Structures)
   > n30 = Mesh Map Merge(Array 1=@n29, Merge Dest Array=Get All StaticMeshes())
   > Set All StaticMeshes(All StaticMeshes=@n30)
   n31 = Get Many Coll Prims Array() -> (Many Coll Prims Array:Map of Static Mesh Object References to S Asset Errors Asset Health Structures)
   Clear(execute=@n6.then_6, TargetMap=@n31)
   > n32 = MeshCheck_ManyCollPrims!(
    execute
    Static Mesh:Static Mesh Object Reference=Get Cur Project StaticMesh()
    AcceptableCollPrims:int=Get Acceptable Coll Prims() -> (Acceptable Coll Prims:int=0)
    ManyCollPrims_Array:Map of Static Mesh Object References to S Asset Errors Asset Health Structures=@n31
    Only Major:bool=true
   ) -> (then, Updated Array:Map of Static Mesh Object References to S Asset Errors Asset Health Structures)
   > n33 = Set Many Coll Prims Array(Many Coll Prims Array:Map of Static Mesh Object References to S Asset Errors Asset Health Structures=@n32) -> (Output_Get:Map of Static Mesh Object References to S Asset Errors Asset Health Structures)
   > n34 = Mesh Map Merge(Array 1=@n33, Merge Dest Array=Get All StaticMeshes())
   > Set All StaticMeshes(All StaticMeshes=@n34)
   n35 = Get LOD0 Collision Array() -> (LOD0 Collision Array:Map of Static Mesh Object References to S Asset Errors Asset Health Structures)
   Clear(execute=@n6.then_8, TargetMap=@n35)
   > n36 = MeshCheck_LOD0_Collision!(
    execute
    Static Mesh:Static Mesh Object Reference=Get Cur Project StaticMesh()
    LOD0_CollisionArray:Map of Static Mesh Object References to S Asset Errors Asset Health Structures=@n35
    Only Major:bool=true
   ) -> (then, Updated Array:Map of Static Mesh Object References to S Asset Errors Asset Health Structures)
   > n37 = Set LOD0 Collision Array(LOD0 Collision Array:Map of Static Mesh Object References to S Asset Errors Asset Health Structures=@n36) -> (Output_Get:Map of Static Mesh Object References to S Asset Errors Asset Health Structures)
   > n38 = Mesh Map Merge(Array 1=@n37, Merge Dest Array=Get All StaticMeshes())
   > Set All StaticMeshes(All StaticMeshes=@n38)
   Outputs()
  ]
 }
 Find all assets to check {
  path=On Run Tests Clicked.K2Node_Composite_2.Find all assets to check
  nodes [
   n1 = Get Project Asset Path() -> (Project Asset Path:String)
   n2 = GetAssetsByPath!(
    execute
    self:Asset Registry Interface=GetAssetRegistry()
    PackagePath:Name=String To Name(InString=@n1)
    bRecursive:bool=true
    bIncludeOnlyOnDiskAssets:bool=false
   ) -> (then, OutAssetData:Array of Asset Data Structures, ReturnValue:bool)
   > n3 = For Each Loop(Array:Array of Asset Data Structures=@n2) -> (Array Element:Asset Data Structure)
   > Branch(Condition="Equal (Class)"(A=GetClass!(InAssetData:"Asset Data Structure (by ref)"=@n3) -> (ReturnValue:Object Class Reference), B=&StaticMesh))
   > n4 = Cast To StaticMesh(Object=GetAsset(InAssetData=@n3))
   > Add(TargetArray:Array of Static Mesh Object References=Get Project Static Mesh(), NewItem:"Static Mesh Object Reference (by ref)"=@n4) {is=title K2Node_CallArrayFunction}
   Inputs()
   > n5 = GetProjectAssetsWithFiltering!(execute, ProjPath:String=@n1, Filter:String=DO_NOT_OPTIMIZE) -> (then, Found Static Meshes:Array of Static Mesh Object References, Found Textures:Array of Texture 2D Object References)
   > Set Project Static Mesh(Project Static Mesh:Array of Static Mesh Object References=@n5) -> (Output_Get:Array of Static Mesh Object References)
   Outputs()
  ]
 }
 Sort into Severity {
  path=On Run Tests Clicked.K2Node_Composite_3.Sort into Severity
  nodes [
   Inputs!() -> (InputPin:Exec)
   > n1 = Sequence!(execute) -> (then_0, then_1)
   > Set Cur Mesh Asset Selected(Cur Mesh Asset Selected=0) -> (Output_Get=0)
   Clear(
    execute=@n1.then_1
    TargetMap=Get StaticMeshes Severe() -> (StaticMeshes Severe:Map of Static Mesh Object References to S Asset Errors Asset Health Structures)
   )
   > Clear(TargetMap=Get StaticMeshes Moderate() -> (StaticMeshes Moderate:Map of Static Mesh Object References to S Asset Errors Asset Health Structures))
   > Clear(TargetMap=Get StaticMeshes Minor() -> (StaticMeshes Minor:Map of Static Mesh Object References to S Asset Errors Asset Health Structures))
   > Set Num Severe(Num Severe:int=0) -> (Output_Get:int=0)
   > Set Num Moderate(Num Moderate:int=0) -> (Output_Get:int=0)
   > Set Num Minor(Num Minor:int=0) -> (Output_Get:int=0)
   > n2 = "For Each Loop (Map)"(MapPin=Get All StaticMeshes())
   > Set Cur Static Mesh In Map(Cur Static Mesh In Map:Static Mesh Object Reference=@n2) -> (Output_Get:Static Mesh Object Reference)
   > n3 = Set Cur Errors In Map(Cur Errors In Map:S Asset Errors Asset Health Structure=@n2.ValuePin) -> (Output_Get:S Asset Errors Asset Health Structure)
   n4 = Length(TargetArray:Array of Texts=Break S Asset Errors Asset Health(S_AssetErrors_AssetHealth=@n3))
   > n5 = Branch(Condition="integer > integer"(A=@n4, B=2))
   > Increment Int(Value=Get Num Severe() -> (Num Severe:int=0))
   > Add(
    TargetMap=Get StaticMeshes Severe()
    Key=Get Cur Static Mesh In Map() -> (Cur Static Mesh In Map:Static Mesh Object Reference)
    Value=Get Cur Errors In Map() -> (Cur Errors In Map:S Asset Errors Asset Health Structure)
   ) {is=title K2Node_CallFunction}
   n6 = Branch(execute=@n5.else, Condition="integer > integer"(A=@n4, B=1))
   > Increment Int(Value=Get Num Moderate() -> (Num Moderate:int=0))
   > Add(TargetMap=Get StaticMeshes Moderate(), Key=Get Cur Static Mesh In Map(), Value=Get Cur Errors In Map())
   Increment Int(" "=@n6.else, Value=Get Num Minor() -> (Num Minor:int=0))
   > Add(TargetMap=Get StaticMeshes Minor(), Key=Get Cur Static Mesh In Map(), Value=Get Cur Errors In Map())
   Outputs()
  ]
 }
 Make Severe Collection {
  path=On Run Tests Clicked.K2Node_Composite_4.Make Severe Collection
  nodes [
   Inputs()
   > n1 = Branch(Condition="integer > integer"(A=Get Num Severe() -> (Num Severe=0)))
   > n2 = Make Collection!(
    execute
    Collection Name:String=Get Severe Collection Name() -> (Severe Collection Name:String)
    Parent Collection Name:String=Get Parent Collection Name()
    ShareType:ECollectionScriptingShareType Enum=Get Collection Share Type() -> (Collection Share Type=Local)
    Container:Collection Container Source Structure=Get Valkyrie Main Proj Collection Container()
    Parent Color:Linear Color Structure=Get Parent Color() -> (Parent Color:Linear Color Structure)
    Collection Color:Linear Color Structure=Get Severe Collection Color() -> (Severe Collection Color:Linear Color Structure)
   ) -> (then, OutNewOrEmptyCollection:Collection Structure, Success:bool, OutParentCollection:Collection Structure) {is=title K2Node_CallFunction}
   > Set Severe Collection(Severe Collection:Collection Structure=@n2) -> (Output_Get:Collection Structure)
   > Set Parent Collection(Parent Collection=@n2.OutParentCollection)
   > n3 = "For Each Loop (Map)"(MapPin=Get StaticMeshes Severe())
   > AddAssetToCollection!(
    execute
    self:Collection Manager Subsystem Object Reference=Get CollectionManagerScriptingSubsystem()
    Collection:"Collection Structure (by ref)"=Get Severe Collection()
    AssetPath:"Soft Object Path Structure (by ref)"=GetSoftObjectPath!(Object=@n3) -> (ReturnValue:Soft Object Path Structure)
   ) -> (then, ReturnValue:bool)
   n4 = CollectionExists(
    execute=@n1.else
    self=Get CollectionManagerScriptingSubsystem()
    Container=Get Valkyrie Main Proj Collection Container()
    Collection=String To Name(InString=Get Severe Collection Name())
    ShareType=Get Collection Share Type() -> (Collection Share Type=Local)
   )
   > Branch(Condition=@n4)
   > Destroy Collection!(
    execute
    self
    Collection Name:String=Get Severe Collection Name()
    ShareType:ECollectionScriptingShareType Enum=Get Collection Share Type() -> (Collection Share Type=Local)
   ) -> (then, Success:bool) {is=title K2Node_CallFunction}
   Outputs()
  ]
 }
 Make Moderate Collection {
  path=On Run Tests Clicked.K2Node_Composite_6.Make Moderate Collection
  nodes [
   Inputs!() -> (execute:Exec)
   > n1 = Branch(Condition="integer > integer"(A=Get Num Moderate() -> (Num Moderate=0)))
   > n2 = Make Collection(
    Collection Name=Get Moderate Collection Name() -> (Moderate Collection Name:String)
    Parent Collection Name=Get Parent Collection Name()
    ShareType=Get Collection Share Type() -> (Collection Share Type=Local)
    Container=Get Valkyrie Main Proj Collection Container()
    Parent Color=Get Parent Color()
    Collection Color=Get Moderate Collection Color() -> (Moderate Collection Color:Linear Color Structure)
   )
   > Set Moderate Collection(Moderate Collection:Collection Structure=@n2) -> (Output_Get:Collection Structure)
   > Set Parent Collection(Parent Collection=@n2.OutParentCollection)
   > n3 = "For Each Loop (Map)"(MapPin=Get StaticMeshes Moderate())
   > AddAssetToCollection(self=Get CollectionManagerScriptingSubsystem(), Collection=Get Moderate Collection(), AssetPath=GetSoftObjectPath(Object=@n3))
   n4 = CollectionExists(
    execute=@n1.else
    self=Get CollectionManagerScriptingSubsystem()
    Container=Get Valkyrie Main Proj Collection Container()
    Collection=String To Name(InString=Get Moderate Collection Name())
    ShareType=Get Collection Share Type() -> (Collection Share Type=Local)
   )
   > Branch(Condition=@n4)
   > Destroy Collection(Collection Name=Get Moderate Collection Name(), ShareType=Get Collection Share Type() -> (Collection Share Type=Local))
   Outputs()
  ]
 }
 Make Minor Collection {
  path=On Run Tests Clicked.K2Node_Composite_7.Make Minor Collection
  nodes [
   Inputs()
   > n1 = Branch(Condition="integer > integer"(A=Get Num Minor() -> (Num Minor=0)))
   > n2 = Make Collection(
    Collection Name=Get Minor Collection Name() -> (Minor Collection Name:String)
    Parent Collection Name=Get Parent Collection Name()
    ShareType=Get Collection Share Type() -> (Collection Share Type=Local)
    Container=Get Valkyrie Main Proj Collection Container()
    Parent Color=Get Parent Color()
    Collection Color=Get Minor Collection Color() -> (Minor Collection Color:Linear Color Structure)
   )
   > Set Minor Collection(Minor Collection:Collection Structure=@n2) -> (Output_Get:Collection Structure)
   > Set Parent Collection(Parent Collection=@n2.OutParentCollection)
   > n3 = "For Each Loop (Map)"(MapPin=Get StaticMeshes Minor())
   > AddAssetToCollection(self=Get CollectionManagerScriptingSubsystem(), Collection=Get Minor Collection(), AssetPath=GetSoftObjectPath(Object=@n3))
   n4 = CollectionExists(
    execute=@n1.else
    self=Get CollectionManagerScriptingSubsystem()
    Container=Get Valkyrie Main Proj Collection Container()
    Collection=String To Name(InString=Get Minor Collection Name())
    ShareType=Get Collection Share Type() -> (Collection Share Type=Local)
   )
   > Branch(Condition=@n4)
   > Destroy Collection(Collection Name=Get Minor Collection Name(), ShareType=Get Collection Share Type() -> (Collection Share Type=Local))
   Outputs()
  ]
 }
 Set Menu To highest severity {
  path=On Run Tests Clicked.K2Node_Composite_9.Set Menu To highest severity
  nodes [
   Inputs()
   > n1 = Branch(Condition="integer > integer"(A=Get Num Severe() -> (Num Severe=0)))
   > n2 = Set Assets to Cycle Through!(
    execute
    Assets to Cycle Through:E_CollectionChoices_StaticMeshHealth Enum=NewEnumerator0
    self:BP STPS Static Mesh Health Object Reference=Get Settings()
   ) -> (then, Output_Get:E_CollectionChoices_StaticMeshHealth Enum=NewEnumerator0)
   NextButtonEnable(execute=[@n2.then, @n10.then, @n8.then], Enable=true)
   > n3 = CreateCollectionMessage!(execute, CurSev:E_CollectionChoices_TextureHealth Enum=NewEnumerator0, NumFound:int=Get Num Severe() -> (Num Severe=0), CurrentMessage:Text) -> (then, Message:Text)
   > n4 = CreateCollectionMessage(CurSev=NewEnumerator1, NumFound=Get Num Moderate() -> (Num Moderate=0), CurrentMessage=@n3)
   > n5 = CreateCollectionMessage(CurSev=NewEnumerator2, NumFound=Get Num Minor() -> (Num Minor=0), CurrentMessage=@n4)
   > n6 = Set FinalRunMessage(FinalRunMessage=@n5)
   > DisplayUserWarningMessage(Message=@n6)
   n7 = Branch(execute=@n1.else, Condition="integer > integer"(A=Get Num Moderate() -> (Num Moderate=0)))
   > n8 = Set Assets to Cycle Through(Assets to Cycle Through=NewEnumerator1, self=Get Settings()) -> (Output_Get=NewEnumerator0)
   n9 = Branch(execute=@n7.else, Condition="integer > integer"(A=Get Num Minor() -> (Num Minor=0)))
   > n10 = Set Assets to Cycle Through(Assets to Cycle Through=NewEnumerator2, self=Get Settings()) -> (Output_Get=NewEnumerator0)
   DisplayUserWarningMessage(execute=@n9.else, Message="LOCTEXT(\"No problems found!\")")
   Outputs()
  ]
 }
 On Severity Changed {nodes [
   On Severity Changed!() -> (then, PropertySet:Scriptable Interactive Tool Property Set Object Reference, PropertyName:String) {is=title K2Node_FunctionEntry}
   > Set Cur Mesh Asset Selected(Cur Mesh Asset Selected=0) -> (Output_Get=0)
  ]}
 OnClearCollectionClicked {nodes [
   OnClearCollectionClicked() {is=title K2Node_FunctionEntry}
   > n1 = GetCollections!(
    execute
    self:Collection Manager Subsystem Object Reference=Get CollectionManagerScriptingSubsystem()
    Container:Collection Container Source Structure=Get Valkyrie Main Proj Collection Container()
   ) -> (then, OutCollections:Array of Collection Structures, ReturnValue:bool)
   > n2 = For Each Loop(Array:Array of Collection Structures=Make Array!(
     "[0]":Collection Structure=Get Severe Collection()
     "[1]":Collection Structure=Get Moderate Collection()
     "[2]":Collection Structure=Get Minor Collection()
     "[3]":Collection Structure=Get Parent Collection()
    ) -> (Array:Array of Collection Structures)) -> (Array Element:Collection Structure)
   > Branch(Condition="integer > integer"(A=Find Item!(TargetArray:Array of Collection Structures=@n1, ItemToFind:"Collection Structure (by ref)"=@n2) -> (ReturnValue), B=-1))
   > DestroyCollection(self=Get CollectionManagerScriptingSubsystem(), Collection=@n2)
  ]}
 OnFixButton_AddCollisionClicked {nodes [
   OnFixButton_AddCollisionClicked() {is=title K2Node_FunctionEntry}
   > MeshFix_AddSimpleCollision_NDOP26!(execute, Mesh:Static Mesh Object Reference=Get Cur Mesh For Button Fixes() -> (Cur Mesh For Button Fixes:Static Mesh Object Reference)) -> (then, Fix Successful:bool)
  ]}
 OnFixButton_AddLODsClicked {nodes [
   OnFixButton_AddLODsClicked() {is=title K2Node_FunctionEntry}
   n1 = Get Cur Mesh For Button Fixes()
   > MeshFix_AddLODs!(execute, Mesh:Static Mesh Object Reference=@n1) -> (then, Fix Successful:bool)
   > SaveAndReload!(execute, self, Object:Object Reference=@n1) -> (then)
  ]}
 OnFixButton_CollisionComplexity_Clicked {nodes [
   OnFixButton_CollisionComplexity_Clicked() {is=title K2Node_FunctionEntry}
   n1 = Get Cur Mesh For Button Fixes()
   > MeshFix_CollisionComplexity!(execute, Mesh:Static Mesh Object Reference=@n1) -> (then, Fix Successful:bool)
   > SaveAndReload(Object=@n1)
  ]}
 OnFixButton_SetLODForCollision_Clicked {nodes [
   OnFixButton_SetLODForCollision_Clicked() {is=title K2Node_FunctionEntry}
   n1 = Get Cur Mesh For Button Fixes()
   > MeshFix_SetLODforCollision!(execute, Mesh:Static Mesh Object Reference=@n1) -> (then, Fix Successful:bool)
   > SaveAndReload(Object=@n1)
  ]}
 OnFixButton_SetMinLODforPlatform_Clicked {nodes [
   OnFixButton_SetMinLODforPlatform_Clicked() {is=title K2Node_FunctionEntry}
   n1 = Get Cur Mesh For Button Fixes()
   > MeshFix_SetMinLODforPlatform!(execute, Mesh:Static Mesh Object Reference=@n1) -> (then, Fix Successful:bool)
   > SaveAndReload(Object=@n1)
  ]}
 OnFixButton_TurnOnNaniteClicked {nodes [
   OnFixButton_TurnOnNaniteClicked() {is=title K2Node_FunctionEntry}
   n1 = Get Cur Mesh For Button Fixes()
   > MeshFix_TurnOnNanite!(execute, Mesh:Static Mesh Object Reference=@n1) -> (then, Fix Successful:bool)
   > SaveAndReload(Object=@n1)
  ]}
 Populate Collection Vars {nodes [
   Populate Collection Vars() {is=title K2Node_FunctionEntry}
   > n1 = Sequence!(execute) -> (then_0, then_1, then_2:Exec)
   > Set Cur Mesh Asset Selected(Cur Mesh Asset Selected=0) -> (Output_Get=0)
   > Clear(TargetMap=Get All StaticMeshes())
   n2 = GetCollections(execute=@n1.then_1, self=Get CollectionManagerScriptingSubsystem(), Container=Get Valkyrie Main Proj Collection Container())
   > n3 = For Each Loop(Array=@n2)
   > n4 = Set Cur Found Array(Cur Found Array:Collection Structure=@n3) -> (Output_Get:Collection Structure)
   n5 = Break Collection(CollectionScriptingRef=@n4)
   n6 = "To String (Name)"(InName=@n5.Name)
   n7 = Get Parent Collection Name()
   n8 = Get Share Type() -> (Share Type:ECollectionScriptingShareType Enum=Local)
   > n9 = Branch(Condition=AND Boolean(
     A="Equal Exactly (String)"(A=@n7, B=@n6)
     B="Equal (Enum)"(A:ECollectionScriptingShareType Enum=@n5.ShareType, B:ECollectionScriptingShareType Enum=@n8)
    ))
   > Set Parent Collection(Parent Collection=Get Cur Found Array() -> (Cur Found Array:Collection Structure))
   n10 = Break Collection(CollectionScriptingRef=Get Cur Found Array())
   n11 = "To String (Name)"(InName=@n10.Name)
   n12 = Get Severe Collection Name()
   n13 = Get Share Type() -> (Share Type=Local)
   n14 = Branch(execute=@n9.else, Condition=AND Boolean(A="Equal Exactly (String)"(A=@n12, B=@n11), B="Equal (Enum)"(A=@n10.ShareType, B=@n13)))
   > Set Severe Collection(Severe Collection=Get Cur Found Array())
   n15 = Break Collection(CollectionScriptingRef=Get Cur Found Array())
   n16 = "To String (Name)"(InName=@n15.Name)
   n17 = Get Moderate Collection Name()
   n18 = Get Share Type() -> (Share Type=Local)
   n19 = Branch(execute=@n14.else, Condition=AND Boolean(A="Equal Exactly (String)"(A=@n17, B=@n16), B="Equal (Enum)"(A=@n15.ShareType, B=@n18)))
   > Set Moderate Collection(Moderate Collection=Get Cur Found Array())
   n20 = Break Collection(CollectionScriptingRef=Get Cur Found Array())
   n21 = "To String (Name)"(InName=@n20.Name)
   n22 = Get Minor Collection Name()
   n23 = Get Share Type() -> (Share Type=Local)
   Branch(execute=@n19.else, Condition=AND Boolean(A="Equal Exactly (String)"(A=@n22, B=@n21), B="Equal (Enum)"(A=@n20.ShareType, B=@n23)))
   > Set Minor Collection(Minor Collection=Get Cur Found Array())
   n24 = Get Collection Colors!(execute=@n1.then_2) -> (
    then
    Common Parent Color:Linear Color Structure
    "Common  Severe Collection Color":Linear Color Structure
    Common Moderate Collection Color:Linear Color Structure
    Common Minor Collection Color:Linear Color Structure
   ) {is=title K2Node_CallFunction}
   > Set Parent Color(Parent Color:Linear Color Structure=@n24) -> (Output_Get:Linear Color Structure)
   > Set Severe Collection Color(Severe Collection Color:Linear Color Structure=@n24."Common  Severe Collection Color") -> (Output_Get:Linear Color Structure)
   > Set Moderate Collection Color(Moderate Collection Color:Linear Color Structure=@n24.Common Moderate Collection Color) -> (Output_Get:Linear Color Structure)
   > Set Minor Collection Color(Minor Collection Color:Linear Color Structure=@n24.Common Minor Collection Color) -> (Output_Get:Linear Color Structure)
  ]}
 SaveAndReload {nodes [
   n1 = SaveAndReload!() -> (then, Object:Object Reference) {is=title K2Node_FunctionEntry}
   > SaveDirtyPackagesWithDialog!(execute, bSaveMapPackages:bool=false, bSaveContentPackages:bool=true) -> (then, ReturnValue:bool)
   > ReloadPackages!(
    execute
    PackagesToReload:Array of Package Object References=Make Array!("[0]":Package Object Reference=GetPackageForObject!(Object:Object Reference=@n1) -> (ReturnValue:Package Object Reference)) -> (Array:Array of Package Object References)
    InteractionMode:EReloadPackagesInteractionMode Enum=Interactive
   ) -> (then, bOutAnyPackagesReloaded:bool, OutErrorMessage:Text)
  ]}
]
variables [
 Acceptable Coll Prims {type=integer, default=50}
 Acceptable Min LOD Count {type=integer, default=2}
 Acceptable Num Mat Slots {type=integer, default=3}
 Acceptable Num Mat Slots_0 {type=integer, default=3}
 Acceptable Percent Diff For LODs {type=integer, default=50}
 Actions Box {type=BP_UW_ExpandableBox_C}
 All StaticMeshes {}
 Collection Share Type {type=ECollectionScriptingShareType, default=Local}
 ComplexAsSimple {}
 Cur Mesh Asset Selected {type=integer}
 Cur Mesh For Button Fixes {type=StaticMesh, default=null}
 Currently Selected Asset {type=StaticMesh, default=null}
 CurTime {type=number}
 FinalRunMessage {type=string, default=""}
 FirstInitRun {type=boolean}
 Has LODS Array {}
 Header {type=BP_UW_SharedHeaderWidget_C}
 LOD0 Collision Array {}
 LOD0 Too Many Verts {}
 Main Project Handle {type=ValkyrieProjectHandle, default {}}
 Many Coll Prims Array {}
 Min LOD for Low {type=integer, default=2}
 Min LOD for Medium {type=integer, default=1}
 Min LODfor Low Array {}
 Minor Collection {type=CollectionScriptingRef, default {name=Minor, container=Game, shareType=Local}}
 Minor Collection Color {type=LinearColor, default=<rgba 1 0.9143 0 1>}
 Minor Collection Name {type=string, default=Minor}
 Moderate Collection {type=CollectionScriptingRef, default {name=Moderate, container=Game, shareType=Local}}
 Moderate Collection Color {type=LinearColor, default=<rgba 1 0.2263 0 1>}
 Moderate Collection Name {type=string, default=Moderate}
 Nanite Not Enabled {}
 NextButtonDisabledText {type=string, default=Run Tests to cycle to next static mesh.}
 NextButtonEnabledText {type=string, default=Cycles to the next static mesh in the collection.}
 NextStaticMeshButton {type=BP_UW_NextStaticMeshButton_C, default=null}
 Not Enough LODs {}
 Num Mat Slots Array {}
 Num Mat Slots Collision Array {}
 Num Minor {type=integer}
 Num Moderate {type=integer}
 Num Severe {type=integer}
 Parent Collection {type=CollectionScriptingRef, default {name=None, container=Game, shareType=Local}}
 Parent Collection Name {type=string, default=StaticMesh_Health}
 Parent Color {type=LinearColor, default=<rgba 0.8788 0 1 1>}
 Previously Selected Asset {type=StaticMesh, default=null}
 Project Asset Path {type=string, default=""}
 Project Handle Valid {type=boolean}
 Prop Set Save Key {type=string, default=SaveSettings}
 RunTestsButton {type=BP_UW_RunTestsButton_C, default=null}
 Settings {type=BP_STPS_StaticMeshHealth_C, default=null}
 Severe Collection {type=CollectionScriptingRef, default {name=Severe, container=Game, shareType=Local}}
 Severe Collection Color {type=LinearColor, default=<rgba 1 0 0.0355 1>}
 Severe Collection Name {type=string, default=Severe}
 Share Type {type=ECollectionScriptingShareType, default=Local}
 Similar LODS Array {}
 StaticMeshes Minor {}
 StaticMeshes Moderate {}
 StaticMeshes Severe {}
 ThisAssetButtons {type="array<BP_UW_Shared_FixMeButton_C>", default []}
 Too Many LODs {}
 Valkyrie Main Proj Collection Container {type=CollectionScriptingContainerSource, default {name=Game, title=Collections}}
]
